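# EQS query that picks spawn spots for guards at the mid-match Radio Tower objective, judging by its name.
# A Blueprint BossHenchmenRiftSpawn generator, reused here, supplies candidate actors around the querier.
# A 2D distance test rejects candidates too close to any player, using a scalable-float minimum, and its negative factor favors nearer survivors.
# A safe-zone test affects score only, without filtering.
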EnvQuery /MidMatchRadioTowerGameplay/EQS/GuardSpawner/EQS_MMObj_RadioTowerGuardSpawn
options [{
  generator {
   name=EnvQueryGenerator_GuardSpawn_C_0
   class=EnvQueryGenerator_BossHenchmenRiftSpawn_C
   properties {
    bAutoSortTests=true
    bCanRunAsync=false
    context=&EnvQueryContext_Querier
    envQueryResultNormalizationOption=Default
    generatedItemType=/Script/AIModule.EnvQueryItemType_Actor
    generatorsActionDescription=""
    optionName=""
   }
  }
  tests [
   EnvQueryTest_Distance_0 {
    class=EnvQueryTest_Distance
    properties {
     distanceTo=/Script/FortniteAIServer.FortQueryContext_AllPlayers
     filterType=Minimum
     scoringEquation=Linear
     scoringFactor=-1
     testComment=""
     testMode=Distance2D
     testPurpose=FilterAndScore
     floatValueMin {
      dataBinding=&self:EnvQueryTest_Distance_0.FortAIDataProvider_ScalableFloat_0
      dataField=FloatValue
      defaultValue=1500
     }
    }
   }
   FortQueryTest_InsideAthenaSafeZone_1 {class=FortQueryTest_InsideAthenaSafeZone, properties {testPurpose=Score}}
  ]
 }]
